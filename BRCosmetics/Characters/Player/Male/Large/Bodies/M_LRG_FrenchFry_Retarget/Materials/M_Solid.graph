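# M_Solid is an unlit skeletal-mesh material in the cosmetics folder for the large male FrenchFry retarget player body.
# Its graph multiplies vertex color by a scalar Color Mult parameter that slightly darkens it by default.
# It samples no textures, and the file shows no connection from the multiply to a material output.
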Material /BRCosmetics/Characters/Player/Male/Large/Bodies/M_LRG_FrenchFry_Retarget/Materials/M_Solid
graph {
 properties {bAutomaticallySetUsageInEditor=false, bUsedWithSkeletalMesh=true, shadingModel=MSM_Unlit, usedShadingModels=Unlit}
 nodes [Multiply(A=VertexColor(), B=ScalarParameter() {defaultValue=0.8, parameterName=Color Mult}) {constB=0.8}]
}
